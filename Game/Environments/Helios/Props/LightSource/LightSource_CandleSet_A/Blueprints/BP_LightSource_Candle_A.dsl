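; Candle light source prop Blueprint whose flame varies in appearance from instance to instance.
; The construction script runs the parent's setup, then uses the instance random stream to randomize width, distortion strength, bias and distortion tiling on the flame mesh's materials.
; Exposed asset provider overrides extend the parent, adding a version entry and registering the resolved candle material as an allowed indirect reference.
; The event graph holds only disabled default stubs.
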(event EventBeginPlay
  (|CallParentFunction) ; This node is disabled and will not be called.\nDrag off pins to build functionality.
)

(event Collision|EventActorBeginOverlap
  (params (OtherActor :type Actor%20Object%20Reference))
  (|CallParentFunction OtherActor) ; This node is disabled and will not be called.\nDrag off pins to build functionality.
)

(event EventTick
  (params (DeltaSeconds :type Float%20%28single-precision%29))
  (|CallParentFunction DeltaSeconds) ; This node is disabled and will not be called.\nDrag off pins to build functionality.
)

(fn BuildExposedAssetProviderVersion(Implementation)
  (params (Version :type Fort%20Exposed%20Asset%20Provider%20Version%20Structure%20%28by%20ref%29))
  (|CallParentFunction Version)
  (ExposedAssetProvider|AddVersion Version 1)
  (return))

(fn CollectAssetReferencestoAllow(Implementation)
  (params (Collector :type Fort%20Exposed%20Asset%20Reference%20Collector%20Structure%20%28by%20ref%29))
  (|CallParentFunction Collector)
  (ExposedAssetProvider|CollectIndirectAssetReference Collector (Utilities|ResolveSoftReference (LocalVariables|GetCandleMaterial)))
  (return))

(fn ConstructionScript
  (bind _sm_spline_candleflame_0_0 (Variables|Default|GetSM_Spline_CandleFlame_0_0))
  (bind _instancerandomstream (Variables|Default|GetInstanceRandomStream))
  (|CallParentFunction)
  (Rendering|Material|SetScalarParameterValueonMaterials _sm_spline_candleflame_0_0 "Width" (Math|Random|RandomFloatinRangefromStream _instancerandomstream 2.0 5.0))
  (Rendering|Material|SetScalarParameterValueonMaterials _sm_spline_candleflame_0_0 "Distortionstrength" (Math|Random|RandomFloatinRangefromStream _instancerandomstream 0.4 0.75))
  (Rendering|Material|SetScalarParameterValueonMaterials _sm_spline_candleflame_0_0 "Bias" (Math|Random|RandomFloatinRangefromStream _instancerandomstream 0.0 0.6))
  (Rendering|Material|SetScalarParameterValueonMaterials _sm_spline_candleflame_0_0 "DistortionTilingMult" (Math|Random|RandomFloatinRangefromStream _instancerandomstream 1.0 1.4)))
