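; Skull candle light-source prop Blueprint that gives its candle flame a randomized look.
; The construction script runs the parent's script, then sets random width, distortion strength, bias and distortion tiling scalars on the flame mesh component's materials.
; The random calls use no seed stream, so the flame can change whenever construction reruns.
; BeginPlay, overlap and Tick are disabled stubs with no logic of their own.
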(event EventBeginPlay
  (|CallParentFunction) ; This node is disabled and will not be called.\nDrag off pins to build functionality.
)

(event Collision|EventActorBeginOverlap
  (params (OtherActor :type Actor%20Object%20Reference))
  (|CallParentFunction OtherActor) ; This node is disabled and will not be called.\nDrag off pins to build functionality.
)

(event EventTick
  (params (DeltaSeconds :type Float%20%28single-precision%29))
  (|CallParentFunction DeltaSeconds) ; This node is disabled and will not be called.\nDrag off pins to build functionality.
)

(fn ConstructionScript
  (bind _sm_spline_candleflame_0_0 (Variables|Default|GetSM_Spline_CandleFlame_0_0))
  (|CallParentFunction)
  (Rendering|Material|SetScalarParameterValueonMaterials _sm_spline_candleflame_0_0 "Width" (Math|Random|RandomFloatinRange 2.0 5.0))
  (Rendering|Material|SetScalarParameterValueonMaterials _sm_spline_candleflame_0_0 "Distortionstrength" (Math|Random|RandomFloatinRange 0.4 0.75))
  (Rendering|Material|SetScalarParameterValueonMaterials _sm_spline_candleflame_0_0 "Bias" (Math|Random|RandomFloatinRange 0.0 0.6))
  (Rendering|Material|SetScalarParameterValueonMaterials _sm_spline_candleflame_0_0 "DistortionTilingMult" (Math|Random|RandomFloatinRange 1.0 1.4)))
